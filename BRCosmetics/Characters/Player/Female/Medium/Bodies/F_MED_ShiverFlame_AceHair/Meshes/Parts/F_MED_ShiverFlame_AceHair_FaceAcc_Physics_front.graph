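# Physics asset for the front braids of the ShiverFlame AceHair female medium body's FaceAcc part, likely driving their secondary motion.
# Left and right five-capsule braid chains hang from small master bodies linked to spine_05, with tight twist limits and per-segment swing limits.
# Capsules on the upper spine, neck, clavicles and upper arms form the torso shapes around the braids.
# Only the right clavicle and upper arm have constraints, both angularly free.
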PhysicsAsset /BRCosmetics/Characters/Player/Female/Medium/Bodies/F_MED_ShiverFlame_AceHair/Meshes/Parts/F_MED_ShiverFlame_AceHair_FaceAcc_Physics_front
bodies {
 clavicle_l [{
   shapeType=Capsule
   length=16.9
   radius=5
   shapeName=clavicle_l_capsule
   center=<3.4338 -2.149 -1.8987>
   rotation=<rot 88.0656 -179.8693 -179.7779>
  }]
 clavicle_r [{
   shapeType=Capsule
   length=16.9
   radius=5
   shapeName=clavicle_l_capsule
   center=<-3.4338 2.149 1.8987>
   rotation=<rot 88.0656 -179.8693 0.2221>
  }]
 dyn_braid_01_l [{shapeType=Capsule, length=2.1, radius=0.8, center=<0.7687 0 0.041>, rotation=<rot 90 0 0>}]
 dyn_braid_01_r [{shapeType=Capsule, length=2.2, radius=0.7, center=<0.6952 0 -0.041>, rotation=<rot 90 0 180>}]
 dyn_braid_02_l [{shapeType=Capsule, length=2.5, radius=1.35, center=<1.8035 0 0>, rotation=<rot 90 0 0>}]
 dyn_braid_02_r [{shapeType=Capsule, length=2.6, radius=1.25, center=<1.755 0 0>, rotation=<rot 90 0 180>}]
 dyn_braid_03_l [{shapeType=Capsule, length=3.05, radius=1.3, center=<2.0299 0 0>, rotation=<rot 90 0 0>}]
 dyn_braid_03_r [{shapeType=Capsule, length=3.15, radius=1.2, center=<1.9844 0 0>, rotation=<rot 90 0 180>}]
 dyn_braid_04_l [{shapeType=Capsule, length=3.35, radius=1.3, center=<2.1277 0 0>, rotation=<rot 90 0 0>}]
 dyn_braid_04_r [{shapeType=Capsule, length=3.45, radius=1.2, center=<2.2267 0 0>, rotation=<rot 90 0 180>}]
 dyn_braid_05_l [{shapeType=Capsule, length=5, radius=1.2, center=<3.5308 0 0>, rotation=<rot 90 0 0>}]
 dyn_braid_05_r [{shapeType=Capsule, length=5.4, radius=1.1, center=<3.4293 0 0>, rotation=<rot 90 0 180>}]
 mstr_dyn_braid_01_l [{shapeType=Capsule, length=0.505, radius=0.505, shapeName=mstr_dyn_braid_01_l_capsule, rotation=<rot 0 -90 0>}]
 mstr_dyn_braid_01_r [{shapeType=Capsule, length=0.505, radius=0.505, shapeName=mstr_dyn_braid_01_l_capsule, rotation=<rot 0 -90 180>}]
 neck_01 [{shapeType=Capsule, length=7.3, radius=4.7, shapeName=neck_01_capsule, center=<2.0971 0 0>, rotation=<rot 90 0 0>}]
 spine_04 [{shapeType=Capsule, length=27.2945, radius=29.1055, shapeName=spine_04_capsule, center=<28.6905 -29.1483 0.3206>}]
 spine_05 [{shapeType=Capsule, length=3.9, radius=36.1, shapeName=spine_05_capsule, center=<-21.0803 -18.2348 0>, rotation=<rot 0 -121.8266 90>}]
 upperarm_l [{
   shapeType=Capsule
   length=5.9
   radius=5.9
   shapeName=upperarm_l_capsule
   center=<-1.154 -0.2466 6.0741>
   rotation=<rot 0.0507 -0.9837 -87.0532>
  }]
 upperarm_r [
  {shapeType=Capsule, length=14.2, radius=3.6, shapeName=upperarm_l_capsule, center=<-11.1571 0 0.8377>, rotation=<rot -84.7888 0.0007 -0.0007>}
  {
   shapeType=Capsule
   length=4.4
   radius=4.7
   shapeName=upperarm_r_capsule
   center=<-0.757 -1.7133 -3.0153>
   rotation=<rot 18.7615 -61.0208 100.1001>
  }
 ]
}
constraints [
 {bone1Name=dyn_braid_01_l, bone2Name=mstr_dyn_braid_01_l, swing1LimitDegrees=60, swing2LimitDegrees=25, twistLimitDegrees=5}
 {bone1Name=dyn_braid_02_l, bone2Name=dyn_braid_01_l, swing1LimitDegrees=40, swing2LimitDegrees=20, twistLimitDegrees=5}
 {bone1Name=dyn_braid_03_l, bone2Name=dyn_braid_02_l, swing1LimitDegrees=30, swing2LimitDegrees=30, twistLimitDegrees=5}
 {bone1Name=dyn_braid_04_l, bone2Name=dyn_braid_03_l, swing1LimitDegrees=30, swing2LimitDegrees=30, twistLimitDegrees=5}
 {bone1Name=dyn_braid_05_l, bone2Name=dyn_braid_04_l, swing1LimitDegrees=40, swing2LimitDegrees=40, twistLimitDegrees=5}
 {
  bone1Name=mstr_dyn_braid_01_l
  bone2Name=spine_05
  swing1LimitDegrees=45
  swing1Motion=Free
  swing2LimitDegrees=45
  swing2Motion=Free
  twistLimitDegrees=45
  twistMotion=Free
 }
 {
  bone1Name=mstr_dyn_braid_01_r
  bone2Name=spine_05
  swing1LimitDegrees=45
  swing1Motion=Free
  swing2LimitDegrees=45
  swing2Motion=Free
  twistLimitDegrees=45
  twistMotion=Free
 }
 {bone1Name=dyn_braid_01_r, bone2Name=mstr_dyn_braid_01_r, swing1LimitDegrees=60, swing2LimitDegrees=25, twistLimitDegrees=5}
 {bone1Name=dyn_braid_02_r, bone2Name=dyn_braid_01_r, swing1LimitDegrees=40, swing2LimitDegrees=20, twistLimitDegrees=5}
 {bone1Name=dyn_braid_03_r, bone2Name=dyn_braid_02_r, swing1LimitDegrees=30, swing2LimitDegrees=30, twistLimitDegrees=5}
 {bone1Name=dyn_braid_04_r, bone2Name=dyn_braid_03_r, swing1LimitDegrees=30, swing2LimitDegrees=30, twistLimitDegrees=5}
 {bone1Name=dyn_braid_05_r, bone2Name=dyn_braid_04_r, swing1LimitDegrees=40, swing2LimitDegrees=40, twistLimitDegrees=5}
 {
  bone1Name=upperarm_r
  bone2Name=clavicle_r
  swing1LimitDegrees=45
  swing1Motion=Free
  swing2LimitDegrees=45
  swing2Motion=Free
  twistLimitDegrees=45
  twistMotion=Free
 }
 {
  bone1Name=clavicle_r
  bone2Name=spine_05
  swing1LimitDegrees=45
  swing1Motion=Free
  swing2LimitDegrees=45
  swing2Motion=Free
  twistLimitDegrees=45
  twistMotion=Free
 }
]
