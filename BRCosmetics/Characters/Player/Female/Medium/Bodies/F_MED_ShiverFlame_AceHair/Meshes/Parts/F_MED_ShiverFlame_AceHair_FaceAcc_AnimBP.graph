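# Animation Blueprint for the ShiverFlame AceHair face accessory, adding secondary motion to the character's hair.
# It copies the parent pose, applies corrective bone rotations and upper-arm transforms, then caches it for RigidBody simulation.
# Blends select a front-end physics asset or, in gameplay, skydive-dive versus normal gravity setups, feeding a long per-bone AnimDynamics spring chain.
# The event graph updates gravity override and samples upper-arm rotations from the body mesh.
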AnimBlueprint /BRCosmetics/Characters/Player/Female/Medium/Bodies/F_MED_ShiverFlame_AceHair/Meshes/Parts/F_MED_ShiverFlame_AceHair_FaceAcc_AnimBP
target_skeleton=&/F_MED_ShiverFlame_AceHair_FaceAcc_Skeleton
graphs [
 AnimGraph {nodes [
   Anim Dynamics - Bone: dyn_FrontBang01_01_r(Alpha=1, ExternalForce="0,0,0") {
    angularDampingOverride=0.99
    angularSpringConstant=2000
    bAngularSpring=true
    bGravityOverrideInSimSpace=true
    bOverrideAngularDamping=true
    bUseGravityOverride=true
    alphaBoolBlend {blendOption=Linear}
    boundBone {boneName=dyn_FrontBang01_01_r}
    componentAppliedLinearAccClamp=<3000 3000 3000>
    componentLinearAccScale=<0.3 0.3 0.3>
    componentLinearVelScale=<0.3 0.3 0.3>
    gravityOverride {z=-980}
    physicsBodyDefinitions [{
      boundBone {boneName=dyn_FrontBang01_01_r}
      boxExtents=<20 20 20>
      constraintSetup {angularLimitsMax=<5 -12 -25>, angularLimitsMin=<-5 -20 -30>, angularTarget=<0 0 0>}
      localJointOffset=<1.5 0 0>
     }]
   }
   Anim Dynamics - Bone: mstr_dyn_BackStrand01_01_l(Alpha=1, ExternalForce="0,0,0") {
    angularDampingOverride=0.99
    angularSpringConstant=2000
    bAngularSpring=true
    bGravityOverrideInSimSpace=true
    bOverrideAngularDamping=true
    bUseGravityOverride=true
    alphaBoolBlend {blendOption=Linear}
    boundBone {boneName=mstr_dyn_BackStrand01_01_l}
    componentAppliedLinearAccClamp=<3000 3000 3000>
    componentLinearAccScale=<0.3 0.3 0.3>
    componentLinearVelScale=<0.3 0.3 0.3>
    gravityOverride {z=-980}
    physicsBodyDefinitions [{
      boundBone {boneName=mstr_dyn_BackStrand01_01_l}
      boxExtents=<40 40 40>
      constraintSetup {angularLimitsMax=<1 -3 5>, angularLimitsMin=<0 0 -20>, angularTarget=<0 0 0>}
      localJointOffset=<1 0 0>
     }]
   }
   n1 = RigidBody!(ComponentPose=Local To Component(LocalPose=Use cached pose 'BeforeRban'()), Alpha, OverrideWorldGravity:Vector="0,0,0", ExternalForce) -> (Pose) {
    bClampLinearTranslationLimitToRefPose=true
    bOverrideWorldGravity=true
    overridePhysicsAsset=&/F_MED_ShiverFlame_AceHair_FaceAcc_Physics
    simulationSpace=BaseBoneSpace
    alphaBoolBlend {blendOption=Linear}
    baseBoneRef {boneName=neck_01}
    componentAppliedLinearAccClamp=<3000 3000 3000>
    componentLinearAccScale=<0.4 0.4 0.5>
    componentLinearVelScale=<0.1 0.1 0.1>
   }
   n2 = RigidBody(ComponentPose=Local To Component(LocalPose=Use cached pose 'BeforeRban'()), OverrideWorldGravity="0,0,-980") {
    bClampLinearTranslationLimitToRefPose=true
    bOverrideWorldGravity=true
    overridePhysicsAsset=&F_MED_ShiverFlame_AceHair_FaceAcc_Physics
    simulationSpace=BaseBoneSpace
    alphaBoolBlend {blendOption=Linear}
    baseBoneRef {boneName=neck_01}
    componentAppliedLinearAccClamp=<3000 3000 3000>
    componentLinearAccScale=<0.4 0.4 0.5>
    componentLinearVelScale=<0.1 0.1 0.1>
   }
   n3 = Blend Poses by bool(
    BlendPose_0=Component To Local(ComponentPose=@n1)
    BlendPose_1=Component To Local(ComponentPose=@n2)
    BlendTime_0=0.9
    BlendTime_1=0.9
    bActiveValue=Get bIsSkydiveDiveMode()
   ) {blendType=Linear}
   n4 = RigidBody(ComponentPose=Local To Component(LocalPose=Use cached pose 'BeforeRban'()), OverrideWorldGravity="0,0,-980") {
    bClampLinearTranslationLimitToRefPose=true
    bOverrideWorldGravity=true
    overridePhysicsAsset=&/F_MED_ShiverFlame_AceHair_FaceAcc_Physics_front
    simulationSpace=BaseBoneSpace
    alphaBoolBlend {blendOption=Linear}
    baseBoneRef {boneName=neck_01}
    componentAppliedLinearAccClamp=<3000 3000 3000>
    componentLinearAccScale=<0.4 0.4 0.5>
    componentLinearVelScale=<0.1 0.1 0.1>
   }
   n5 = Component To Local(ComponentPose=@n4)
   n6 = Get bIsInFrontEnd()
   n7 = Anim Dynamics - Bone: dyn_FrontBang01_01_l(
    ComponentPose=Local To Component(LocalPose=Blend Poses by bool(BlendPose_0=@n5, BlendPose_1=@n3, BlendTime_0=0.9, BlendTime_1=0.9, bActiveValue=@n6) {blendType=Linear})
    Alpha=1
    ExternalForce="0,0,0"
   ) {
    angularDampingOverride=0.99
    angularSpringConstant=1000
    bAngularSpring=true
    bGravityOverrideInSimSpace=true
    bOverrideAngularDamping=true
    bUseGravityOverride=true
    alphaBoolBlend {blendOption=Linear}
    boundBone {boneName=dyn_FrontBang01_01_l}
    componentAppliedLinearAccClamp=<3000 3000 3000>
    componentLinearAccScale=<0.3 0.3 0.3>
    componentLinearVelScale=<0.3 0.3 0.3>
    gravityOverride {z=-980}
    physicsBodyDefinitions [{
      boundBone {boneName=dyn_FrontBang01_01_l}
      boxExtents=<20 20 20>
      constraintSetup {angularLimitsMax=<5 0 0>, angularLimitsMin=<-5 -15 -9>, angularTarget=<0 0 0>}
      localJointOffset=<1.5 0 0>
     }]
   }
   n8 = Anim Dynamics - Bone: dyn_FrontBang01_02_l(ComponentPose=@n7, Alpha=1, ExternalForce="0,0,0") {
    angularDampingOverride=0.95
    angularSpringConstant=800
    bAngularSpring=true
    bGravityOverrideInSimSpace=true
    bOverrideAngularDamping=true
    bUseGravityOverride=true
    alphaBoolBlend {blendOption=Linear}
    boundBone {boneName=dyn_FrontBang01_02_l}
    componentAppliedLinearAccClamp=<3000 3000 3000>
    componentLinearAccScale=<0.3 0.3 0.3>
    componentLinearVelScale=<0.3 0.3 0.3>
    gravityOverride {z=-980}
    physicsBodyDefinitions [{
      boundBone {boneName=dyn_FrontBang01_02_l}
      boxExtents=<15 15 15>
      constraintSetup {angularLimitsMax=<5 -8 60>, angularLimitsMin=<-5 -40 -30>, angularTarget=<0 0 0>}
      localJointOffset=<1.5 0 0>
     }]
   }
   n9 = Anim Dynamics - Bone: dyn_FrontBang01_03_l(ComponentPose=@n8, Alpha=1, ExternalForce="0,0,0") {
    angularDampingOverride=0.99
    angularSpringConstant=800
    bAngularSpring=true
    bGravityOverrideInSimSpace=true
    bOverrideAngularDamping=true
    bUseGravityOverride=true
    alphaBoolBlend {blendOption=Linear}
    boundBone {boneName=dyn_FrontBang01_03_l}
    componentAppliedLinearAccClamp=<3000 3000 3000>
    componentLinearAccScale=<0.3 0.3 0.3>
    componentLinearVelScale=<0.3 0.3 0.3>
    gravityOverride {z=-980}
    physicsBodyDefinitions [{
      boundBone {boneName=dyn_FrontBang01_03_l}
      boxExtents=<13 13 13>
      constraintSetup {angularLimitsMax=<5 10 60>, angularLimitsMin=<-5 -40 -30>, angularTarget=<0 0 0>}
      localJointOffset=<1.5 0 0>
     }]
   }
   n10 = Anim Dynamics - Bone: dyn_FrontBang02_01_l(ComponentPose=@n9, Alpha=1, ExternalForce="0,0,0") {
    angularDampingOverride=0.99
    angularSpringConstant=1000
    bAngularSpring=true
    bGravityOverrideInSimSpace=true
    bOverrideAngularDamping=true
    bUseGravityOverride=true
    alphaBoolBlend {blendOption=Linear}
    boundBone {boneName=dyn_FrontBang02_01_l}
    componentAppliedLinearAccClamp=<3000 3000 3000>
    componentLinearAccScale=<0.3 0.3 0.3>
    componentLinearVelScale=<0.3 0.3 0.3>
    gravityOverride {z=-980}
    physicsBodyDefinitions [{
      boundBone {boneName=dyn_FrontBang02_01_l}
      boxExtents=<20 20 20>
      constraintSetup {angularLimitsMax=<5 0 15>, angularLimitsMin=<-5 -45 -15>, angularTarget=<0 0 0>}
      localJointOffset=<2 0 0>
     }]
   }
   n11 = Anim Dynamics - Bone: dyn_FrontBang02_02_l(ComponentPose=@n10, Alpha=1, ExternalForce="0,0,0") {
    angularDampingOverride=0.95
    angularSpringConstant=800
    bAngularSpring=true
    bGravityOverrideInSimSpace=true
    bOverrideAngularDamping=true
    bUseGravityOverride=true
    alphaBoolBlend {blendOption=Linear}
    boundBone {boneName=dyn_FrontBang02_02_l}
    componentAppliedLinearAccClamp=<3000 3000 3000>
    componentLinearAccScale=<0.3 0.3 0.3>
    componentLinearVelScale=<0.3 0.3 0.3>
    gravityOverride {z=-980}
    physicsBodyDefinitions [{
      boundBone {boneName=dyn_FrontBang02_02_l}
      boxExtents=<13 13 13>
      constraintSetup {angularLimitsMax=<5 0 15>, angularLimitsMin=<-5 -20 -15>, angularTarget=<0 0 0>}
      localJointOffset=<2 0 0>
     }]
   }
   n12 = Anim Dynamics - Bone: dyn_FrontBang03_01_l(ComponentPose=@n11, Alpha=1, ExternalForce="0,0,0") {
    angularDampingOverride=0.99
    angularSpringConstant=1000
    bAngularSpring=true
    bGravityOverrideInSimSpace=true
    bOverrideAngularDamping=true
    bUseGravityOverride=true
    alphaBoolBlend {blendOption=Linear}
    boundBone {boneName=dyn_FrontBang03_01_l}
    componentAppliedLinearAccClamp=<3000 3000 3000>
    componentLinearAccScale=<0.3 0.3 0.3>
    componentLinearVelScale=<0.3 0.3 0.3>
    gravityOverride {z=-980}
    physicsBodyDefinitions [{
      boundBone {boneName=dyn_FrontBang03_01_l}
      boxExtents=<20 20 20>
      constraintSetup {angularLimitsMax=<0 0 5>, angularLimitsMin=<0 0 -5>, angularTarget=<0 0 0>}
      localJointOffset=<1.5 0 0>
     }]
   }
   n13 = Anim Dynamics - Bone: dyn_FrontBang03_02_l(ComponentPose=@n12, Alpha=1, ExternalForce="0,0,0") {
    angularDampingOverride=0.99
    angularSpringConstant=800
    bAngularSpring=true
    bGravityOverrideInSimSpace=true
    bOverrideAngularDamping=true
    bUseGravityOverride=true
    alphaBoolBlend {blendOption=Linear}
    boundBone {boneName=dyn_FrontBang03_02_l}
    componentAppliedLinearAccClamp=<3000 3000 3000>
    componentLinearAccScale=<0.3 0.3 0.3>
    componentLinearVelScale=<0.3 0.3 0.3>
    gravityOverride {z=-980}
    physicsBodyDefinitions [{
      boundBone {boneName=dyn_FrontBang03_02_l}
      boxExtents=<18 18 15>
      constraintSetup {angularLimitsMax=<0 0 20>, angularLimitsMin=<0 -10 -20>, angularTarget=<0 0 0>}
      localJointOffset=<1.6 0 0>
     }]
   }
   n14 = Anim Dynamics - Bone: dyn_FrontBang03_03_l(ComponentPose=@n13, Alpha=1, ExternalForce="0,0,0") {
    angularDampingOverride=0.99
    angularSpringConstant=800
    bAngularSpring=true
    bGravityOverrideInSimSpace=true
    bOverrideAngularDamping=true
    bUseGravityOverride=true
    alphaBoolBlend {blendOption=Linear}
    boundBone {boneName=dyn_FrontBang03_03_l}
    componentAppliedLinearAccClamp=<3000 3000 3000>
    componentLinearAccScale=<0.3 0.3 0.3>
    componentLinearVelScale=<0.3 0.3 0.3>
    gravityOverride {z=-980}
    physicsBodyDefinitions [{
      boundBone {boneName=dyn_FrontBang03_03_l}
      boxExtents=<14 15 13>
      constraintSetup {angularLimitsMax=<0 0 20>, angularLimitsMin=<0 -5 -20>, angularTarget=<0 0 0>}
      localJointOffset=<1.45 0 0>
     }]
   }
   n15 = Anim Dynamics - Bone: dyn_FrontBang01_02_r(ComponentPose=@n14, Alpha=1, ExternalForce="0,0,0") {
    angularDampingOverride=0.99
    angularSpringConstant=2000
    bAngularSpring=true
    bGravityOverrideInSimSpace=true
    bOverrideAngularDamping=true
    bUseGravityOverride=true
    alphaBoolBlend {blendOption=Linear}
    boundBone {boneName=dyn_FrontBang01_02_r}
    componentAppliedLinearAccClamp=<3000 3000 3000>
    componentLinearAccScale=<0.3 0.3 0.3>
    componentLinearVelScale=<0.3 0.3 0.3>
    gravityOverride {z=-980}
    physicsBodyDefinitions [{
      boundBone {boneName=dyn_FrontBang01_02_r}
      boxExtents=<15 15 15>
      constraintSetup {angularLimitsMax=<5 5 5>, angularLimitsMin=<-5 20 -10>, angularTarget=<0 0 0>}
      localJointOffset=<1.5 0 0>
     }]
   }
   n16 = Anim Dynamics - Bone: dyn_FrontBang01_03_r(ComponentPose=@n15, Alpha=1, ExternalForce="0,0,0") {
    angularDampingOverride=0.99
    angularSpringConstant=2000
    bAngularSpring=true
    bGravityOverrideInSimSpace=true
    bOverrideAngularDamping=true
    bUseGravityOverride=true
    alphaBoolBlend {blendOption=Linear}
    boundBone {boneName=dyn_FrontBang01_03_r}
    componentAppliedLinearAccClamp=<3000 3000 3000>
    componentLinearAccScale=<0.3 0.3 0.3>
    componentLinearVelScale=<0.3 0.3 0.3>
    gravityOverride {z=-980}
    physicsBodyDefinitions [{
      boundBone {boneName=dyn_FrontBang01_03_r}
      boxExtents=<13 13 13>
      constraintSetup {angularLimitsMax=<5 5 5>, angularLimitsMin=<-5 -20 -15>, angularTarget=<0 0 0>}
      localJointOffset=<1.5 0 0>
     }]
   }
   n17 = Anim Dynamics - Bone: dyn_FrontBang02_01_r(ComponentPose=@n16, Alpha=1, ExternalForce="0,0,0") {
    angularDampingOverride=0.99
    angularSpringConstant=1000
    bAngularSpring=true
    bGravityOverrideInSimSpace=true
    bOverrideAngularDamping=true
    bUseGravityOverride=true
    alphaBoolBlend {blendOption=Linear}
    boundBone {boneName=dyn_FrontBang02_01_r}
    componentAppliedLinearAccClamp=<3000 3000 3000>
    componentLinearAccScale=<0.3 0.3 0.3>
    componentLinearVelScale=<0.3 0.3 0.3>
    gravityOverride {z=-980}
    physicsBodyDefinitions [{
      boundBone {boneName=dyn_FrontBang02_01_r}
      boxExtents=<20 20 20>
      constraintSetup {angularLimitsMax=<5 -5 5>, angularLimitsMin=<-5 -7 -20>, angularTarget=<0 0 0>}
      localJointOffset=<2 0 0>
     }]
   }
   n18 = Anim Dynamics - Bone: dyn_FrontBang02_02_r(ComponentPose=@n17, Alpha=1, ExternalForce="0,0,0") {
    angularDampingOverride=0.99
    angularSpringConstant=800
    bAngularSpring=true
    bGravityOverrideInSimSpace=true
    bOverrideAngularDamping=true
    bUseGravityOverride=true
    alphaBoolBlend {blendOption=Linear}
    boundBone {boneName=dyn_FrontBang02_02_r}
    componentAppliedLinearAccClamp=<3000 3000 3000>
    componentLinearAccScale=<0.3 0.3 0.3>
    componentLinearVelScale=<0.3 0.3 0.3>
    gravityOverride {z=-980}
    physicsBodyDefinitions [{
      boundBone {boneName=dyn_FrontBang02_02_r}
      boxExtents=<13 13 13>
      constraintSetup {angularLimitsMax=<5 0 15>, angularLimitsMin=<-5 -7 -25>, angularTarget=<0 0 0>}
      localJointOffset=<2 0 0>
     }]
   }
   n19 = Anim Dynamics - Bone: dyn_SideBang01_01_r(ComponentPose=@n18, Alpha=1, ExternalForce="0,0,0") {
    angularDampingOverride=0.99
    angularSpringConstant=3000
    bAngularSpring=true
    bGravityOverrideInSimSpace=true
    bOverrideAngularDamping=true
    bUseGravityOverride=true
    bUsePlanarLimit=false
    alphaBoolBlend {blendOption=Linear}
    boundBone {boneName=dyn_SideBang01_01_r}
    componentAppliedLinearAccClamp=<3000 3000 3000>
    componentLinearAccScale=<0.3 0.3 0.3>
    componentLinearVelScale=<0.3 0.3 0.3>
    gravityOverride {z=-980}
    physicsBodyDefinitions [{
      boundBone {boneName=dyn_SideBang01_01_r}
      boxExtents=<15 15 12>
      constraintSetup {angularLimitsMax=<5 0 -20>, angularLimitsMin=<-5 -1 -25>, angularTarget=<0 0 0>}
      localJointOffset=<1.5 0 0>
     }]
    planarLimits [{drivingBone {boneName=upperarm_r}, planeTransform {location=<0 0 0>, rotation=<rot 0 0 -90>}}]
   }
   n20 = Anim Dynamics - Bone: dyn_SideBang01_02_r(ComponentPose=@n19, Alpha=1, ExternalForce="0,0,0") {
    angularDampingOverride=0.99
    angularSpringConstant=3000
    bAngularSpring=true
    bGravityOverrideInSimSpace=true
    bOverrideAngularDamping=true
    bUseGravityOverride=true
    bUsePlanarLimit=false
    alphaBoolBlend {blendOption=Linear}
    boundBone {boneName=dyn_SideBang01_02_r}
    componentAppliedLinearAccClamp=<3000 3000 3000>
    componentLinearAccScale=<0.3 0.3 0.3>
    componentLinearVelScale=<0.3 0.3 0.3>
    gravityOverride {z=-980}
    physicsBodyDefinitions [{
      boundBone {boneName=dyn_SideBang01_02_r}
      boxExtents=<12 12 12>
      constraintSetup {angularLimitsMax=<5 0 -5>, angularLimitsMin=<-5 -30 -20>, angularTarget=<0 0 0>}
      localJointOffset=<1.5 0 0>
     }]
    planarLimits [{drivingBone {boneName=upperarm_r}, planeTransform {location=<0 0 0>, rotation=<rot 0 0 -90>}}]
   }
   n21 = Anim Dynamics - Bone: dyn_SideBang02_01_l(ComponentPose=@n20, Alpha=1, ExternalForce="0,0,0") {
    angularDampingOverride=0.99
    angularSpringConstant=3000
    bAngularSpring=true
    bGravityOverrideInSimSpace=true
    bOverrideAngularDamping=true
    bUseGravityOverride=true
    alphaBoolBlend {blendOption=Linear}
    boundBone {boneName=dyn_SideBang02_01_l}
    componentAppliedLinearAccClamp=<3000 3000 3000>
    componentLinearAccScale=<0.3 0.3 0.3>
    componentLinearVelScale=<0.3 0.3 0.3>
    gravityOverride {z=-980}
    physicsBodyDefinitions [{
      boundBone {boneName=dyn_SideBang02_01_l}
      boxExtents=<20 20 20>
      constraintSetup {angularLimitsMax=<5 0 20>, angularLimitsMin=<-5 -30 -5>, angularTarget=<0 0 0>}
      localJointOffset=<2 0 0>
     }]
   }
   n22 = Anim Dynamics - Bone: dyn_SideBang02_02_l(ComponentPose=@n21, Alpha=1, ExternalForce="0,0,0") {
    angularDampingOverride=0.99
    angularSpringConstant=3000
    bAngularSpring=true
    bGravityOverrideInSimSpace=true
    bOverrideAngularDamping=true
    bUseGravityOverride=true
    alphaBoolBlend {blendOption=Linear}
    boundBone {boneName=dyn_SideBang02_02_l}
    componentAppliedLinearAccClamp=<3000 3000 3000>
    componentLinearAccScale=<0.3 0.3 0.3>
    componentLinearVelScale=<0.3 0.3 0.3>
    gravityOverride {z=-980}
    physicsBodyDefinitions [{
      boundBone {boneName=dyn_SideBang02_02_l}
      boxExtents=<13 13 13>
      constraintSetup {angularLimitsMax=<5 0 20>, angularLimitsMin=<-5 -30 -5>, angularTarget=<0 0 0>}
      localJointOffset=<2 0 0>
     }]
   }
   n23 = Anim Dynamics - Bone: dyn_SideBang02_Curl_r(ComponentPose=@n22, Alpha=1, ExternalForce="0,0,0") {
    angularDampingOverride=0.99
    angularSpringConstant=5000
    bAngularSpring=true
    bGravityOverrideInSimSpace=true
    bOverrideAngularDamping=true
    bUseGravityOverride=true
    bUseSphericalLimits=true
    alphaBoolBlend {blendOption=Linear}
    boundBone {boneName=dyn_SideBang02_Curl_r}
    componentAppliedLinearAccClamp=<1000 1000 1000>
    componentLinearAccScale=<0.6 0.5 0.5>
    componentLinearVelScale=<0.6 0.5 0.5>
    gravityOverride {z=-980}
    physicsBodyDefinitions [{
      boundBone {boneName=dyn_SideBang02_Curl_r}
      boxExtents=<20 20 20>
      constraintSetup {angularLimitsMax=<10 10 5>, angularLimitsMin=<-10 -10 -30>, angularTarget=<1 0 0>}
      localJointOffset=<0.5 0 0>
     }]
   }
   n24 = Anim Dynamics - Bone: dyn_SideBang03_Curl_l(ComponentPose=@n23, Alpha=1, ExternalForce="0,0,0") {
    angularDampingOverride=0.99
    angularSpringConstant=3500
    bAngularSpring=true
    bGravityOverrideInSimSpace=true
    bOverrideAngularDamping=true
    bUseGravityOverride=true
    alphaBoolBlend {blendOption=Linear}
    boundBone {boneName=dyn_SideBang03_Curl_l}
    componentAppliedLinearAccClamp=<1000 1000 1000>
    componentLinearAccScale=<0.6 0.5 0.5>
    componentLinearVelScale=<0.6 0.5 0.5>
    gravityOverride {z=-980}
    physicsBodyDefinitions [{
      boundBone {boneName=dyn_SideBang03_Curl_l}
      boxExtents=<15 15 20>
      constraintSetup {angularLimitsMax=<10 10 5>, angularLimitsMin=<-10 -10 -30>, angularTarget=<1 0 0>}
      localJointOffset=<1.5 0 0>
     }]
   }
   n25 = Anim Dynamics - Bone: dyn_BackStrand01_01_l(ComponentPose=@n24, Alpha=1, ExternalForce="0,0,0") {
    angularDampingOverride=0.99
    angularSpringConstant=2000
    bAngularSpring=true
    bGravityOverrideInSimSpace=true
    bOverrideAngularDamping=true
    bUseGravityOverride=true
    alphaBoolBlend {blendOption=Linear}
    boundBone {boneName=dyn_BackStrand01_01_l}
    componentAppliedLinearAccClamp=<3000 3000 3000>
    componentLinearAccScale=<0.3 0.3 0.3>
    componentLinearVelScale=<0.3 0.3 0.3>
    gravityOverride {z=-980}
    physicsBodyDefinitions [{
      boundBone {boneName=dyn_BackStrand01_01_l}
      boxExtents=<20 20 20>
      constraintSetup {angularLimitsMax=<5 0 15>, angularLimitsMin=<-5 -40 -20>, angularTarget=<0 0 0>}
      localJointOffset=<1 0 0>
     }]
   }
   n26 = Anim Dynamics - Bone: dyn_BackStrand01_02_l(ComponentPose=@n25, Alpha=1, ExternalForce="0,0,0") {
    angularDampingOverride=0.95
    angularSpringConstant=2000
    bAngularSpring=true
    bGravityOverrideInSimSpace=true
    bOverrideAngularDamping=true
    bUseGravityOverride=true
    alphaBoolBlend {blendOption=Linear}
    boundBone {boneName=dyn_BackStrand01_02_l}
    componentAppliedLinearAccClamp=<3000 3000 3000>
    componentLinearAccScale=<0.3 0.3 0.3>
    componentLinearVelScale=<0.3 0.3 0.3>
    gravityOverride {z=-950}
    physicsBodyDefinitions [{
      boundBone {boneName=dyn_BackStrand01_02_l}
      boxExtents=<13 13 13>
      constraintSetup {angularLimitsMax=<5 5 20>, angularLimitsMin=<-5 -20 -20>, angularTarget=<0 0 0>}
      localJointOffset=<1 0 0>
     }]
   }
   n27 = Anim Dynamics - Bone: mstr_dyn_BackStrand01_01_r(ComponentPose=@n26, Alpha=1, ExternalForce="0,0,0") {
    angularDampingOverride=0.99
    angularSpringConstant=1000
    bAngularSpring=true
    bGravityOverrideInSimSpace=true
    bOverrideAngularDamping=true
    bUseGravityOverride=true
    alphaBoolBlend {blendOption=Linear}
    boundBone {boneName=mstr_dyn_BackStrand01_01_r}
    componentAppliedLinearAccClamp=<3000 3000 3000>
    componentLinearAccScale=<0.3 0.3 0.3>
    componentLinearVelScale=<0.3 0.3 0.3>
    gravityOverride {z=-980}
    physicsBodyDefinitions [{
      boundBone {boneName=mstr_dyn_BackStrand01_01_r}
      boxExtents=<30 20 30>
      constraintSetup {angularLimitsMax=<5 40 30>, angularLimitsMin=<-5 0 -15>, angularTarget=<0 0 0>}
      localJointOffset=<1.5 0 0>
     }]
   }
   n28 = Anim Dynamics - Bone: dyn_BackStrand01_02_r(ComponentPose=@n27, Alpha=1, ExternalForce="0,0,0") {
    angularDampingOverride=0.95
    angularSpringConstant=800
    bAngularSpring=true
    bGravityOverrideInSimSpace=true
    bOverrideAngularDamping=true
    bUseGravityOverride=true
    alphaBoolBlend {blendOption=Linear}
    boundBone {boneName=dyn_BackStrand01_02_r}
    componentAppliedLinearAccClamp=<3000 3000 3000>
    componentLinearAccScale=<0.3 0.3 0.3>
    componentLinearVelScale=<0.3 0.3 0.3>
    gravityOverride {z=-950}
    physicsBodyDefinitions [{
      boundBone {boneName=dyn_BackStrand01_02_r}
      boxExtents=<13 13 13>
      constraintSetup {angularLimitsMax=<5 5 30>, angularLimitsMin=<-5 -20 -20>, angularTarget=<0 0 0>}
      localJointOffset=<2 0 0>
     }]
   }
   n29 = Anim Dynamics - Bone: dyn_Inside01_l(ComponentPose=@n28, Alpha=1, ExternalForce="0,0,0") {
    angularDampingOverride=0.99
    angularSpringConstant=3000
    bAngularSpring=true
    bGravityOverrideInSimSpace=true
    bOverrideAngularDamping=true
    bUseGravityOverride=true
    alphaBoolBlend {blendOption=Linear}
    boundBone {boneName=dyn_Inside01_l}
    componentAppliedLinearAccClamp=<3000 3000 3000>
    componentLinearAccScale=<0.3 0.3 0.3>
    componentLinearVelScale=<0.3 0.3 0.3>
    gravityOverride {z=-980}
    physicsBodyDefinitions [{
      boundBone {boneName=dyn_Inside01_l}
      boxExtents=<13 13 13>
      constraintSetup {angularLimitsMax=<5 0 15>, angularLimitsMin=<-5 -5 0>, angularTarget=<0 0 0>}
      localJointOffset=<1.5 0 0>
     }]
   }
   n30 = Anim Dynamics - Bone: dyn_Inside02_l(ComponentPose=@n29, Alpha=1, ExternalForce="0,0,0") {
    angularDampingOverride=0.95
    angularSpringConstant=3000
    bAngularSpring=true
    bGravityOverrideInSimSpace=true
    bOverrideAngularDamping=true
    bUseGravityOverride=true
    alphaBoolBlend {blendOption=Linear}
    boundBone {boneName=dyn_Inside02_l}
    componentAppliedLinearAccClamp=<3000 3000 3000>
    componentLinearAccScale=<0.3 0.3 0.3>
    componentLinearVelScale=<0.3 0.3 0.3>
    gravityOverride {z=-950}
    physicsBodyDefinitions [{
      boundBone {boneName=dyn_Inside02_l}
      boxExtents=<13 13 13>
      constraintSetup {angularLimitsMax=<5 5 20>, angularLimitsMin=<-5 -20 -20>, angularTarget=<0 0 0>}
      localJointOffset=<2 0 0>
     }]
   }
   n31 = Anim Dynamics - Bone: dyn_Inside01_r(ComponentPose=@n30, Alpha=1, ExternalForce="0,0,0") {
    angularDampingOverride=0.99
    angularSpringConstant=3000
    bAngularSpring=true
    bGravityOverrideInSimSpace=true
    bOverrideAngularDamping=true
    bUseGravityOverride=true
    alphaBoolBlend {blendOption=Linear}
    boundBone {boneName=dyn_Inside01_r}
    componentAppliedLinearAccClamp=<3000 3000 3000>
    componentLinearAccScale=<0.3 0.3 0.3>
    componentLinearVelScale=<0.3 0.3 0.3>
    gravityOverride {z=-980}
    physicsBodyDefinitions [{
      boundBone {boneName=dyn_Inside01_r}
      boxExtents=<13 13 13>
      constraintSetup {angularLimitsMax=<5 40 30>, angularLimitsMin=<-5 0 -15>, angularTarget=<0 0 0>}
      localJointOffset=<1.5 0 0>
     }]
   }
   n32 = Anim Dynamics - Bone: dyn_Inside02_r(ComponentPose=@n31, Alpha=1, ExternalForce="0,0,0") {
    angularDampingOverride=0.95
    angularSpringConstant=3000
    bAngularSpring=true
    bGravityOverrideInSimSpace=true
    bOverrideAngularDamping=true
    bUseGravityOverride=true
    alphaBoolBlend {blendOption=Linear}
    boundBone {boneName=dyn_Inside02_r}
    componentAppliedLinearAccClamp=<3000 3000 3000>
    componentLinearAccScale=<0.3 0.3 0.3>
    componentLinearVelScale=<0.3 0.3 0.3>
    gravityOverride {z=-950}
    physicsBodyDefinitions [{
      boundBone {boneName=dyn_Inside02_r}
      boxExtents=<13 13 13>
      constraintSetup {angularLimitsMax=<5 5 30>, angularLimitsMin=<-5 -20 -20>, angularTarget=<0 0 0>}
      localJointOffset=<2 0 0>
     }]
   }
   Output Pose(Result=Component To Local(ComponentPose=@n32)) {layerGroup=None}
   n33 = "Transform (Modify) Bone - Bone: mstr_dyn_FrontBang03_01_l"(
    ComponentPose=Local To Component(LocalPose=Copy Pose From Mesh() {bUseAttachedParent=true})
    Alpha=1
    Translation="0,0,0"
    Rotation="-2.5,0,0"
    Scale="1,1,1"
   ) {
    rotationMode=BMM_Additive
    rotationSpace=BCS_BoneSpace
    alphaBoolBlend {blendOption=Linear}
    boneToModify {boneName=mstr_dyn_FrontBang03_01_l}
    rotation {pitch=-2.5}
   }
   n34 = "Transform (Modify) Bone - Bone: dyn_Inside01_l"(ComponentPose=@n33, Alpha=1, Translation="0,0,0", Rotation="-0.168,2.057,6.981", Scale="1,1,1") {
    rotationMode=BMM_Additive
    rotationSpace=BCS_BoneSpace
    alphaBoolBlend {blendOption=Linear}
    boneToModify {boneName=dyn_Inside01_l}
    rotation=<rot -0.168 2.057 6.981>
   }
   n35 = "Transform (Modify) Bone - Bone: mstr_dyn_BackStrand01_01_l"(ComponentPose=@n34, Alpha=1, Translation="0,0,0", Rotation="5.267,3.670,-46.918", Scale="1,1,1") {
    rotationMode=BMM_Additive
    rotationSpace=BCS_BoneSpace
    alphaBoolBlend {blendOption=Linear}
    boneToModify {boneName=mstr_dyn_BackStrand01_01_l}
    rotation=<rot 5.267 3.67 -46.918>
   }
   n36 = "Transform (Modify) Bone - Bone: upperarm_l"!(ComponentPose=@n35, Alpha=1, Rotation="0,0,0") -> (Pose) {
    rotationMode=BMM_Replace
    rotationSpace=BCS_WorldSpace
    scaleMode=BMM_Additive
    translationSpace=BCS_WorldSpace
    alphaBoolBlend {blendInTime=1, blendOption=Linear, blendOutTime=1}
    boneToModify {boneName=upperarm_l}
   }
   n37 = "Transform (Modify) Bone - Bone: upperarm_r"!(ComponentPose=@n36, Alpha=1, Rotation="0,0,0") -> (Pose) {
    rotationMode=BMM_Replace
    rotationSpace=BCS_WorldSpace
    scaleMode=BMM_Additive
    translationSpace=BCS_WorldSpace
    alphaBoolBlend {blendInTime=1, blendOption=Linear, blendOutTime=1}
    boneToModify {boneName=upperarm_r}
   }
   Save cached pose 'BeforeRban'(Pose=Component To Local(ComponentPose=@n37)) {cacheName=BeforeRban}
  ]}
 EventGraph {nodes [
   Event BlueprintUpdateAnimation()
   > n1 = Gravity Override State Function DEPRECATED(CharacterPartAnimInstance=Self-Reference(), GameTimeInSeconds=GetGameTimeInSeconds(), GravityOverrideParams=Get Gravity Override Params())
   > Set Gravity Override(Gravity Override=@n1)
   n2 = Get BodyMesh() -> (BodyMesh:Skeletal Mesh Component Object Reference)
   > Is Valid(InputObject=@n2)
   > Set Upperarm_L_Rotation(Upperarm_L_Rotation:Rotator=BreakTransform(InTransform=GetSocketTransform(self=@n2, InSocketName=upperarm_l, TransformSpace=RTS_World)).Rotation) -> (Output_Get:Rotator)
   > Set Upperarm_R_Rotation(Upperarm_R_Rotation:Rotator=BreakTransform(InTransform=GetSocketTransform(self=@n2, InSocketName=upperarm_r, TransformSpace=RTS_World)).Rotation) -> (Output_Get:Rotator)
   n3 = Event OnCharacterCustomizationLoaded()
   > Set BodyMesh(BodyMesh=Get Array Item with EFort Custom Part Type(Array=@n3.CustomMeshParts))
  ]}
]
variables [
 BodyMesh {type=SkeletalMeshComponent}
 Gravity Override {type=Vector}
 Gravity Override Params {type=GravityOverrideParamsStruct}
 Head Up {type=number}
 Upperarm_L_Rotation {type=Rotator}
 Upperarm_R_Rotation {type=Rotator}
]
